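# Default Lit material for the RL170SE vehicle cosmetic wheel, giving rim and tire an animated crystal look.
# Custom and accent colors mix through masks, with static switches and separate rim and tire parameters for color, metallic and roughness.
# Panning crystal masks, parallax BumpOffset patterns and self-distorted triangle shimmer add motion, feeding overlay-blended normals and roughness variation.
# An advanced Fresnel on the blended normal drives tinted emissive.
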Material /VehicleCosmetics/Wheels/RL170SE/Materials/M_Wheel_RL170SE
graph {
 properties {bAutomaticallySetUsageInEditor=false, bUsedWithSkeletalMesh=true, usedShadingModels=Default Lit}
 nodes [
  n1 = VectorParameter() {parameterName=CustomColor, defaultValue=<rgba 0.8796 0.5149 0.2384 1>}
  n2 = ScalarParameter() {defaultValue=0.5, parameterName=Color-de-brighteningRim}
  n3 = LinearInterpolate(
   A=Multiply(A=@n1.RGB, B=@n2)
   B=Multiply(A=ScalarParameter() {defaultValue=0.5, parameterName=Color-de-brighteningTire}, B=@n1.RGB)
   Alpha=NamedRerouteUsage()
  )
  n4 = LinearInterpolate(
   A=ScalarParameter() {defaultValue=0.8, parameterName=DarkenInnerColorsRim}
   B=ScalarParameter() {defaultValue=1, parameterName=DarkenInnerColorsTire}
   Alpha=NamedRerouteUsage()
  )
  n5 = TextureSampleParameter2D(UVs=BumpOffset(Height=NamedRerouteUsage(), HeightRatioInput=Constant() {r=0.03})) {parameterName=RimMask, samplerType=SAMPLERTYPE_LinearColor, texture=/VehicleCosmetics/Wheels/RL170SE/Textures/T_RL170SE_M}
  n6 = NamedRerouteUsage()
  n7 = TextureCoordinate() {coordinateIndex=1, uTiling=0.5}
  n8 = TextureSample(UVs=BumpOffset(Coordinate=Panner(Coordinate=@n7) {speedX=0.025}, Height=Constant() {r=-2})) {texture=/VehicleCosmetics/Wheels/RL170SE/Textures/T_RL170SE_StarliteCrystalPattern}
  n9 = Add(A=Multiply(A=LinearInterpolate(A=@n5.G, B=@n8.R, Alpha=@n6), B=Constant() {r=0.8}), B=NamedRerouteUsage())
  n10 = OneMinus(None=@n9)
  n11 = Multiply(A=VectorParameter() {parameterName=AccentColor, defaultValue=<rgba 0.0222 0.0307 0.0307 1>}.RGB, B=@n2)
  n12 = Multiply(A=Add(A=@n3, B=@n11), B=Constant() {r=0.5})
  n13 = LinearInterpolate(A=Multiply(A=@n12, B=@n4), B=@n12, Alpha=@n10)
  n14 = LinearInterpolate(A=Multiply(A=@n11, B=@n4), B=@n11, Alpha=@n10)
  n15 = StaticSwitchParameter(True=@n13, False=@n14) {defaultValue=true, parameterName=Color2LightRim}
  n16 = Multiply(A=@n15, B=ScalarParameter() {defaultValue=1, parameterName=DarkenSecondaryColorsRim})
  n17 = StaticSwitchParameter(True=@n13, False=@n14) {parameterName=Color2LightTIre}
  n18 = Multiply(A=@n17, B=ScalarParameter() {defaultValue=0.2, parameterName=DarkenSecondaryColorsTire})
  n19 = LinearInterpolate(
   A=LinearInterpolate(A=Multiply(A=@n3, B=@n4), B=@n3, Alpha=@n10)
   B=LinearInterpolate(
    A=StaticSwitchParameter(True=@n16, False=@n15) {defaultValue=true, parameterName=DarkenSecondaryColorRim}
    B=StaticSwitchParameter(True=@n18, False=@n17) {parameterName=DarkenSecondaryColorTire}
    Alpha=NamedRerouteUsage()
   )
   Alpha=NamedRerouteUsage()
  )
  n20 = Constant3Vector()
  n21 = LinearInterpolate(B=NamedRerouteUsage(), Alpha=OneMinus(None=NamedRerouteUsage()))
  LinearInterpolate(A=@n19, B=@n20, Alpha=@n21)
  LinearInterpolate(A=NamedRerouteUsage(), B=Constant() {r=1}, Alpha=@n21)
  n22 = Multiply(A=TextureCoordinate() {coordinateIndex=1}, B=ScalarParameter() {defaultValue=1, parameterName=Tile_N_for_LOD})
  n23 = TextureSample(UVs=BumpOffset(Coordinate=Panner(Coordinate=@n22) {speedX=0.1}, Height=Constant()) {heightRatio=0}) {samplerType=SAMPLERTYPE_Normal, texture=/VehicleCosmetics/Wheels/RL170SE/Textures/T_RL170SE_FacetedPattern_N}
  n24 = TextureSampleParameter2D(UVs=TextureCoordinate()) {parameterName=Rim_N, samplerType=SAMPLERTYPE_Normal, texture=/VehicleCosmetics/Wheels/RL170SE/Textures/T_RL170SE_Rim_N}
  n25 = TextureSample(UVs=Panner() {constCoordinate=1, speedX=-0.0132, speedY=0.0063}) {texture=/VehicleCosmetics/Wheels/FrioCrystal/Textures/wheel_frio_SE_rgb_mask}
  n26 = TextureSample(UVs=Panner() {constCoordinate=1, speedX=0.0073, speedY=-0.01}) {texture=&wheel_frio_SE_rgb_mask}
  n27 = AppendVector(A=Divide(A=AppendVector(A=@n25.G, B=@n26.B), B=Constant() {r=1}), B=Constant() {r=1})
  n28 = VertexColor()
  n29 = MaterialFunctionCall(
   Base=LinearInterpolate(
    A=MaterialFunctionCall(Base=@n24.RGB, Blend=@n23.RGB) {materialFunction=&Blend_Overlay}.Result
    B=@n27
    Alpha=NamedRerouteDeclaration(None=@n28.R) {name=VC_RC, nodeColor=<rgba 1 0 0.1014 1>}
   )
   Blend=@n27
  ) {materialFunction=&Blend_Overlay}
  n30 = NamedRerouteDeclaration(None=LinearInterpolate(
    A=MaterialFunctionCall(Base=@n23.RGB, Blend=@n29.Result) {materialFunction=&Blend_Overlay}.Result
    B=@n29.Result
    Alpha=NamedRerouteUsage()
   )) {name=Normal, nodeColor=<rgba 0.7109 1 0 1>}
  n31 = Constant() {r=1.8}
  n32 = Multiply(
   A=Multiply(A=@n9, B=ScalarParameter() {defaultValue=1, parameterName=EmisiveIntensity})
   B=Multiply(
    A=MaterialFunctionCall!(Input Normal=@n30, Input Sharpness=@n31, Input Two Sided) {materialFunction=/VehicleCosmetics/Wheels/CarCar_Holographic/Materials/MF_Fresnel_Advanced}.Edge
    B=ScalarParameter() {defaultValue=1, parameterName=FresnelEmissiveIntensity}
   )
  )
  n33 = ScalarParameter() {defaultValue=1, parameterName=AO}
  LinearInterpolate(A=Multiply(A=Multiply(A=@n32, B=@n33), B=@n19), B=@n20, Alpha=@n21)
  n34 = TextureSampleParameter2D() {parameterName=RimMask, samplerType=SAMPLERTYPE_LinearColor, texture=&T_RL170SE_M}
  n35 = OneMinus(None=@n34.R)
  n36 = TextureSampleParameter2D(UVs=TextureCoordinate()) {parameterName=TireMask, samplerType=SAMPLERTYPE_LinearColor, texture=/VehicleCosmetics/Wheels/RL170SE/Textures/T_RL170SE_Crystal_RGB}
  NamedRerouteDeclaration(None=LinearInterpolate(A=Multiply(A=@n26.B, B=@n35), B=@n36.B, Alpha=NamedRerouteUsage())) {name=Green, nodeColor=<rgba 0.0078 0 1 1>}
  n37 = ScalarParameter() {defaultValue=0.3, parameterName=RoughnessRim}
  n38 = ScalarParameter() {defaultValue=0.3, parameterName=RoughnessTire}
  n39 = NamedRerouteUsage()
  n40 = Constant() {r=1}
  NamedRerouteDeclaration(None=LinearInterpolate(A=LinearInterpolate(A=@n37, B=@n38, Alpha=@n39), B=Desaturation(None=@n27, Fraction=@n40), Alpha=Constant() {r=0.15})) {name=Roughness, nodeColor=<rgba 1 0 0.0469 1>}
  NamedRerouteDeclaration(None=LinearInterpolate(
    A=ScalarParameter() {defaultValue=1, parameterName=MetallicRim}
    B=ScalarParameter() {defaultValue=0.5, parameterName=MetallicTire}
    Alpha=NamedRerouteUsage()
   )) {name=Metallic, nodeColor=<rgba 0 0.3672 1 1>}
  NamedRerouteDeclaration(None=@n34.B) {name=LODs_Mask, nodeColor=<rgba 0 1 0.8125 1>}
  n41 = TextureCoordinate() {coordinateIndex=1, uTiling=0.1429}
  n42 = OneMinus(None=Add(A=@n25.G, B=@n26.B))
  n43 = ScalarParameter() {defaultValue=0.025, parameterName=Shimmer1SelfDistort}
  n44 = TextureSample(UVs=Subtract(A=Panner(Coordinate=@n41) {speedX=-0.1, speedY=0.025}, B=Multiply(A=@n42, B=@n43))) {texture=/VehicleCosmetics/Wheels/RL170SE/Textures/T_RL170SE_TriangleShimmer}
  n45 = Multiply(A=@n44.R, B=@n8.R)
  NamedRerouteDeclaration(None=LinearInterpolate(
    A=Constant3Vector() {constant=<rgba 1 1 1 0>}
    B=MaterialFunctionCall!(Blend=@n44.R, Base=@n45) {materialFunction=/Engine/Functions/Engine_MaterialFunctions03/Blends/Blend_Screen}.Result
    Alpha=NamedRerouteUsage()
   )) {name=Blue, nodeColor=<rgba 1 0.3281 0 1>}
  NamedRerouteUsage()
  NamedRerouteUsage()
 ]
}
